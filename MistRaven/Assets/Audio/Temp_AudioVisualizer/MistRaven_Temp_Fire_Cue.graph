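# Temporary MistRaven weapon fire cue, mixing close, medium and distant layers, each with its own attenuation.
# Close routes through a third-person fire sound class; medium and distant share a distant-fire class.
# Each layer blends a randomized start shot with an enveloped loop sample that fades in, both pitch-modulated.
# Close and medium loops repeat, but the distant loop player lacks the looping flag; gunfire concurrency sets are assigned.
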SoundCue /MistRaven/Assets/Audio/Temp_AudioVisualizer/MistRaven_Temp_Fire_Cue
root=Mixer_1
nodes [
 {
  id=Attenuation_10
  children [Mixer_6]
  properties {attenuationSettings=&Fort_Weapon_Projectile_Medium_Mono_Athena_Att}
 }
 {
  id=Attenuation_11
  children [Mixer_7]
  properties {attenuationSettings=&Fort_Weapon_Projectile_Distant_Mono_Athena_Att}
 }
 {id=Attenuation_9, children [Mixer_5], properties {attenuationSettings=&Fort_Weapon_Projectile_Close_Athena_Att}}
 {
  id=Enveloper_0
  children [Random_1]
  properties {pitchCurve {editorCurveData {keys [{time=0}]}}, volumeCurve {editorCurveData {keys [<at 0 0>,<at 0.382 0.2859>,<at 0.65 1>]}}}
 }
 {
  id=Enveloper_1
  children [Random_7]
  properties {pitchCurve {editorCurveData {keys [{time=0}]}}, volumeCurve {editorCurveData {keys [<at 0 0>,<at 0.382 0.2859>,<at 0.65 1>]}}}
 }
 {
  id=Enveloper_2
  children [WavePlayer_39]
  properties {pitchCurve {editorCurveData {keys [{time=0}]}}, volumeCurve {editorCurveData {keys [<at 0 0>,<at 0.382 0.2859>,<at 0.65 1>]}}}
 }
 {id=Mixer_1, children [SoundClass_7,SoundClass_6], properties {inputVolume [0.9,1.5]}}
 {id=Mixer_2, children [Attenuation_10, Attenuation_11], properties {inputVolume [1,1]}}
 {id=Mixer_5, children [Modulator_7,Modulator_0], properties {inputVolume [1,0.5]}}
 {id=Mixer_6, children [Modulator_9,Modulator_8], properties {inputVolume [1,0.7]}}
 {id=Mixer_7, children [Modulator_11,Modulator_10], properties {inputVolume [1,1]}}
 {id=Modulator_0, children [Enveloper_0], properties {pitchMax=1.1, pitchMin=1}}
 {id=Modulator_10, children [Enveloper_2], properties {pitchMax=1.1, pitchMin=0.85}}
 {id=Modulator_11, children [Random_8], properties {pitchMax=0.95, pitchMin=1.05}}
 {id=Modulator_7, children [Random_0], properties {pitchMax=0.95, pitchMin=1.05}}
 {id=Modulator_8, children [Enveloper_1], properties {pitchMax=1, pitchMin=0.8}}
 {id=Modulator_9, children [Random_6], properties {pitchMax=0.95, pitchMin=1.05}}
 {id=Random_0, children [WavePlayer_0,WavePlayer_2,WavePlayer_6,WavePlayer_8], properties {weights [1,1,1,1]}}
 {id=Random_1, children [WavePlayer_9, WavePlayer_10], properties {weights [1,1]}}
 {id=Random_6, children [WavePlayer_27, WavePlayer_28, WavePlayer_29, WavePlayer_30, WavePlayer_31], properties {weights [1,1,1,1,1]}}
 {id=Random_7, children [WavePlayer_32, WavePlayer_33], properties {weights [1,1]}}
 {id=Random_8, children [WavePlayer_34, WavePlayer_35, WavePlayer_36, WavePlayer_37, WavePlayer_38], properties {weights [1,1,1,1,1]}}
 {id=SoundClass_6, children [Mixer_2], properties {soundClassOverride=&Fort_WeaponRanged_Fire_Distant}}
 {
  id=SoundClass_7
  children [Attenuation_9]
  properties {soundClassOverride=&Fort_WeaponRanged_Fire_3P_PSM}
 }
 {id=WavePlayer_0, properties {soundWaveAssetPtr=&/WPN_MistRaven_Temp_Fire_Start_Close_01}}
 {
  id=WavePlayer_10
  properties {bLooping=true, soundWaveAssetPtr=&/WPN_MistRaven_Temp_Fire_Loop_Close_02}
 }
 {id=WavePlayer_2, properties {soundWaveAssetPtr=&/WPN_MistRaven_Temp_Fire_Start_Close_02}}
 {id=WavePlayer_27, properties {soundWaveAssetPtr=&/WPN_MistRaven_Temp_Fire_Start_Med_01}}
 {id=WavePlayer_28, properties {soundWaveAssetPtr=&/WPN_MistRaven_Temp_Fire_Start_Med_02}}
 {id=WavePlayer_29, properties {soundWaveAssetPtr=&/WPN_MistRaven_Temp_Fire_Start_Med_03}}
 {id=WavePlayer_30, properties {soundWaveAssetPtr=&/WPN_MistRaven_Temp_Fire_Start_Med_04}}
 {id=WavePlayer_31, properties {soundWaveAssetPtr=&/WPN_MistRaven_Temp_Fire_Start_Med_05}}
 {
  id=WavePlayer_32
  properties {bLooping=true, soundWaveAssetPtr=&/WPN_MistRaven_Temp_Fire_Loop_Med_01}
 }
 {
  id=WavePlayer_33
  properties {bLooping=true, soundWaveAssetPtr=&/WPN_MistRaven_Temp_Fire_Loop_Med_02}
 }
 {id=WavePlayer_34, properties {soundWaveAssetPtr=&/WPN_MistRaven_Temp_Fire_Start_Dist_01}}
 {id=WavePlayer_35, properties {soundWaveAssetPtr=&/WPN_MistRaven_Temp_Fire_Start_Dist_02}}
 {id=WavePlayer_36, properties {soundWaveAssetPtr=&/WPN_MistRaven_Temp_Fire_Start_Dist_03}}
 {id=WavePlayer_37, properties {soundWaveAssetPtr=&/WPN_MistRaven_Temp_Fire_Start_Dist_04}}
 {id=WavePlayer_38, properties {soundWaveAssetPtr=&/WPN_MistRaven_Temp_Fire_Start_Dist_05}}
 {id=WavePlayer_39, properties {soundWaveAssetPtr=&/WPN_MistRaven_Temp_Fire_Loop_Dist_01}}
 {id=WavePlayer_6, properties {soundWaveAssetPtr=&/WPN_MistRaven_Temp_Fire_Start_Close_04}}
 {id=WavePlayer_8, properties {soundWaveAssetPtr=&/WPN_MistRaven_Temp_Fire_Start_Close_05}}
 {
  id=WavePlayer_9
  properties {bLooping=true, soundWaveAssetPtr=&/WPN_MistRaven_Temp_Fire_Loop_Close_01}
 }
]
properties {
 maxDistance=27500
 volumeMultiplier=1
 attenuationOverrides {falloffDistance=5000, attenuationShapeExtents {x=1000}}
 concurrencyOverrides {bLimitToOwner=true, maxCount=3, resolutionRule=StopOldest}
 concurrencySet [&Weapon_Gunfire_Local_Conc, &Weapon_Gunfire_Global_Conc]
}
